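# Physics asset for the F_MED_Sci_Ops_FaceAcc cosmetic, defining secondary motion for a ponytail and three front hair strands.
# The ponytail chains from the head through pony_01 to 03, then forks into 04 and 05, while each front strand hangs separately off the head, all with swing and tight twist limits.
# Head, neck, upper spine and clavicle capsules supply body volumes, linked by fully free angular constraints.
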PhysicsAsset /BRCosmetics/Accessories/Hats/F_MED_Sci_Ops_FaceAcc/Meshes/F_MED_Sci_Ops_FaceAcc_Dyn
bodies {
 clavicle_l [{shapeType=Capsule, length=7.6738, radius=6.8445, center=<5.2322 -2.0813 -0.3771>, rotation=<rot 74.9997 0 0>}]
 clavicle_r [{shapeType=Capsule, length=6.9863, radius=7.532, center=<-4.8953 0.9285 1.6344>, rotation=<rot 74.9997 0 0>}]
 dyn_hair_fr_01 [{shapeType=Capsule, length=2, radius=2, center=<1 0 0>, rotation=<rot 90 0 0>}]
 dyn_hair_fr_02 [{shapeType=Capsule, length=2, radius=2, center=<1 0 0>, rotation=<rot 90 0 0>}]
 dyn_hair_fr_03 [{shapeType=Capsule, length=2, radius=2, center=<1 0 0>, rotation=<rot 90 0 0>}]
 dyn_hair_pony_01 [{shapeType=Capsule, length=2.75, radius=2.75, center=<0.2 -0.0002 0>, rotation=<rot 90 0 0>}]
 dyn_hair_pony_02 [{shapeType=Capsule, length=3.75, radius=3.5, center=<3.9142 0 0>, rotation=<rot 90 0 0>}]
 dyn_hair_pony_03 [
  {shapeType=Capsule, length=5.1202, radius=3.3081, center=<5.9117 -0.0001 0>, rotation=<rot 90 0 0>}
  {shapeType=Capsule, length=3.5, radius=1, center=<9.4775 -0.5607 -0.0181>, rotation=<rot 60.2129 -179.9999 -160.5817>}
  {shapeType=Capsule, length=3.75, radius=0.5, center=<7.8792 1.0557 -1.5594>, rotation=<rot 73.9908 -136.6751 -131.9255>}
 ]
 dyn_hair_pony_04 [
  {shapeType=Capsule, length=5.375, radius=0.75, shapeName=01, center=<5.3146 2.262 2.2114>, rotation=<rot 61.0415 -140.7751 -170.6891>}
  {shapeType=Capsule, length=2.4104, radius=0.5, center=<0.837 -0.0002 0>, rotation=<rot 90 0 0>}
 ]
 dyn_hair_pony_05 [
  {shapeType=Capsule, length=1.5, radius=0.75, center=<3.817 -0.0001 0>, rotation=<rot 90 0 0>}
  {shapeType=Capsule, length=1.5, radius=0.75, center=<4.6426 -2.6382 0.4971>, rotation=<rot 56.841 -170.3129 -140.4848>}
 ]
 head [{shapeType=Capsule, length=0.2177, radius=2.1656, center=<7.9489 1.3902 0.0999>, rotation=<rot 0 10.1398 -76.3504>}]
 neck_02 [{shapeType=Capsule, length=8.5, radius=4.505, rotation=<rot 90 0 0>}]
 spine_05 [
  {shapeType=Capsule, length=12.7812, radius=9.6875, center=<2.7877 0.0113 0>}
  {shapeType=Capsule, length=7.1562, radius=7.25, center=<3.0126 2.9108 -0.0001>, rotation=<rot 90 -4.7636 18.386>}
 ]
}
constraints [
 {bone1Name=dyn_hair_pony_02, bone2Name=dyn_hair_pony_01, swing1LimitDegrees=45, swing2LimitDegrees=45, twistLimitDegrees=5}
 {bone1Name=dyn_hair_pony_03, bone2Name=dyn_hair_pony_02, swing1LimitDegrees=45, swing2LimitDegrees=45, twistLimitDegrees=5}
 {bone1Name=dyn_hair_pony_04, bone2Name=dyn_hair_pony_03, swing1LimitDegrees=60, swing2LimitDegrees=60, twistLimitDegrees=15}
 {bone1Name=dyn_hair_pony_05, bone2Name=dyn_hair_pony_03, swing1LimitDegrees=45, swing2LimitDegrees=45, twistLimitDegrees=5}
 {bone1Name=dyn_hair_fr_01, bone2Name=head, swing1LimitDegrees=45, swing2LimitDegrees=30, twistLimitDegrees=2}
 {bone1Name=dyn_hair_fr_02, bone2Name=head, swing1LimitDegrees=45, swing2LimitDegrees=30, twistLimitDegrees=2}
 {bone1Name=dyn_hair_pony_01, bone2Name=head, swing1LimitDegrees=45, swing2LimitDegrees=45, twistLimitDegrees=5}
 {bone1Name=dyn_hair_fr_03, bone2Name=head, swing1LimitDegrees=45, swing2LimitDegrees=30, twistLimitDegrees=2}
 {
  bone1Name=clavicle_l
  bone2Name=spine_05
  swing1LimitDegrees=45
  swing1Motion=Free
  swing2LimitDegrees=45
  swing2Motion=Free
  twistLimitDegrees=45
  twistMotion=Free
 }
 {
  bone1Name=clavicle_r
  bone2Name=spine_05
  swing1LimitDegrees=45
  swing1Motion=Free
  swing2LimitDegrees=45
  swing2Motion=Free
  twistLimitDegrees=45
  twistMotion=Free
 }
 {
  bone1Name=head
  bone2Name=neck_02
  swing1LimitDegrees=45
  swing1Motion=Free
  swing2LimitDegrees=45
  swing2Motion=Free
  twistLimitDegrees=45
  twistMotion=Free
 }
 {
  bone1Name=neck_02
  bone2Name=spine_05
  swing1LimitDegrees=45
  swing1Motion=Free
  swing2LimitDegrees=45
  swing2Motion=Free
  twistLimitDegrees=45
  twistMotion=Free
 }
]
